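# Two-sided subsurface Nanite foliage material for instanced Asteria corn patches, finished through MF_Curie_Landscape.
# World position offset uses camera-distance sphere masks and a vertex-color hash to collapse leaves toward encoded pivots or sink cards near the camera.
# Color draws on landscape RVT or biome grass colors, optional textures and hue variation, with distance fade masks defined.
# Static switches select 2D-mask, texture and flower variants, and quality switches provide cheaper shading paths.
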Material /Game/Environments/Asteria/Foliage/Plants/Corn/Materials/M_CornPatch_Nanite
graph {
 properties {
  bTangentSpaceNormal=false
  bUseMaterialAttributes=true
  bUsedWithInstancedStaticMeshes=true
  bUsedWithNanite=true
  pixelDepthOffsetMode=PDOM_Legacy
  shadingModel=MSM_Subsurface
  twoSided=true
  usedShadingModels=Subsurface
 }
 nodes [
  n1 = SetMaterialAttributes!(
   MaterialAttributes
   Base Color=NamedRerouteUsage()
   Specular=NamedRerouteUsage()
   Roughness=NamedRerouteUsage()
   Opacity=NamedRerouteUsage()
   Normal=NamedRerouteUsage()
   World Position Offset=NamedRerouteUsage()
   Subsurface Color=NamedRerouteUsage()
   Opacity Mask=StaticSwitchParameter(True=NamedRerouteUsage(), False=Constant()) {parameterName=Is2DMaskMaterial}
  ) {attributeSetTypes [
    #1
    #2
    #3
    #4
    #5
    #6
    #7
    #8
   ]}
  MaterialFunctionCall!(
   In=@n1
   UseHeightFalloff=StaticBool()
   OnFireEmissiveOverride
   HeightFalloffRadius
   HeightFalloffHardness
   VariationMasksMobile
   In_Grass=StaticBool() {value=true}
   Is_Stylised_WE
  ) {materialFunction=/Game/Packages/Fortress_SharedMaterials/Base_Material_Functions/MF_Curie_Landscape}
  n2 = TextureSampleParameter2D() {
   group=Textures
   parameterName=SMR
   samplerSource=SSM_Wrap_WorldGroupSettings
   samplerType=SAMPLERTYPE_Masks
   texture=&T_WhiteMask
   channelNames {a="", b=Roughness, g=ColorSystemMask, r=Specular}
  }
  NamedRerouteDeclaration(None=@n2.Specular) {name=SpecularTexture, nodeColor=<rgba 0.9688 1 0 1>}
  n3 = Constant3Vector() {desc=Looks almost same RVT normal, constant=<rgba 0 0 1 0>}
  n4 = VertexInterpolator(VS=Transform(None=@n3) {transformSourceType=TRANSFORMSOURCE_Instance})
  n5 = TextureSampleParameter2D() {
   group=Textures
   parameterName=Normal
   samplerType=SAMPLERTYPE_Normal
   texture=/Game/Environments/Asteria/Foliage/Grass/Grass2DAutoBake/Grass_Standard_2D_N
  }
  n6 = MaterialFunctionCall(Normal=@n5.RGB, Flatness=ScalarParameter() {group=Normal Settings, parameterName=TextureNormalFlatten}) {materialFunction=&FlattenNormal}
  n7 = StaticSwitchParameter(True=@n6.Result, False=StaticSwitchParameter(True=@n6.Result, False=NamedRerouteUsage()) {parameterName=Is2DMaskMaterial}) {group=Textures, parameterName=UseTextures}
  n8 = AppendVector(A=Constant2Vector() {g=1, r=1}, B=TwoSidedSign())
  n9 = Add(
   A=Transform(None=Multiply(A=@n7, B=@n8))
   B=Multiply(A=ScalarParameter() {defaultValue=4, group=Normal Settings, parameterName=RVTNormalInfluence}, B=@n3)
  )
  n10 = StaticSwitchParameter(True=LinearInterpolate(A=@n9, B=Constant3Vector() {constant=<rgba 0 0 1 0>}, Alpha=NamedRerouteUsage()), False=@n9) {group=FlowerFade, parameterName=UseFlowerFadeToGreen}
  NamedRerouteDeclaration(None=StaticSwitchParameter(
    True=MaterialFunctionCall(BaseNormal=@n4.PS, AdditionalNormal=@n6.Result) {materialFunction=&BlendAngleCorrectedNormals}.Result
    False=Normalize(VectorInput=@n10)
   ) {parameterName=Is2DMaskMaterial}) {name=Normal, nodeColor=<rgba 1 0 0.1406 1>}
  n11 = NamedRerouteDeclaration(None=LinearInterpolate(
    A=ScalarParameter() {defaultValue=0.16, group=Specular, parameterName=Specular_Near}
    B=ScalarParameter() {defaultValue=0.05, parameterName=Specular_Fade}
    Alpha=NamedRerouteUsage()
   )) {name=SpecularLOWQuality, nodeColor=<rgba 0.6641 0 1 1>}
  n12 = Multiply(A=@n11, B=NamedRerouteUsage())
  n13 = NamedRerouteUsage()
  n14 = Multiply(
   A=Multiply(
    A=StaticSwitchParameter(True=@n12, False=@n11) {group=Textures, parameterName=UseTextures}
    B=LinearInterpolate(Alpha=@n13) {constA=0.3, constB=1.2}
   )
   B=NamedRerouteUsage()
  )
  n15 = NamedRerouteUsage()
  NamedRerouteDeclaration(None=FeatureLevelSwitch(Default=QualitySwitch(Default=@n14, Low=@n15, High=@n15, Medium=@n15, Epic=@n14), ES2=@n15, ES3_1=@n15)) {name=Specular, nodeColor=<rgba 0 0.1562 1 1>}
  n16 = NamedRerouteDeclaration(None=LinearInterpolate(
    A=ScalarParameter() {defaultValue=0.75, group=Roughness, parameterName=Roughness_Near, sliderMin=1}
    B=ScalarParameter() {defaultValue=2, group=Roughness, parameterName=Roughness_Far}
    Alpha=NamedRerouteUsage()
   )) {name=RoughnessLowDetail, nodeColor=<rgba 1 0 0.4922 1>}
  n17 = StaticSwitchParameter(True=Multiply(A=@n16, B=NamedRerouteUsage()), False=@n16) {group=Textures, parameterName=UseTextures}
  n18 = NamedRerouteUsage()
  NamedRerouteDeclaration(None=FeatureLevelSwitch(Default=QualitySwitch(Default=@n17, Low=@n18, Medium=@n18), ES2=@n18, ES3_1=@n18)) {name=Roughness, nodeColor=<rgba 0 1 0.5078 1>}
  n19 = Constant()
  n20 = ScalarParameter() {defaultValue=-170, parameterName=CornCardSinkSize}
  n21 = MaterialFunctionCall() {materialFunction=&ObjectScale}
  n22 = MaterialFunctionCall(Float3=VertexColor()) {materialFunction=&BreakOutFloat3Components}
  n23 = Multiply(
   A=Add(
    A=MaterialFunctionCall(X=@n22.R, Y=@n22.R, Z=@n22.R) {materialFunction=&MakeFloat3}.Result
    B=VectorParameter() {group=NaniteWPO, parameterName=PivotOffsetMatch, defaultValue=<rgba -0.5 -0.5 -0.25 1>}.RGB
   )
   B=ScalarParameter() {defaultValue=1100, group=NaniteWPO, parameterName=ScaleMatch}
  )
  n24 = LinearInterpolate(
   A=MaterialFunctionCall(X=@n19, Y=@n19, Z=Multiply(A=@n20, B=@n21.Scale X)) {materialFunction=&MakeFloat3}.Result
   B=Subtract(
    A=TransformPosition(None=@n23) {transformSourceType=TRANSFORMPOSSOURCE_Instance, transformType=TRANSFORMPOSSOURCE_World}
    B=WorldPosition().XYZ
   )
   Alpha=NamedRerouteUsage()
  )
  n25 = Distance(A=CameraPositionWS(), B=LinearInterpolate(A=NamedRerouteUsage(), B=WorldPosition().XYZ, Alpha=NamedRerouteUsage()))
  n26 = ScalarParameter() {group=NaniteWPO, parameterName=DistanceCardBias}
  n27 = Divide(
   A=LinearInterpolate(A=Add(A=@n25, B=@n26), B=@n25, Alpha=NamedRerouteUsage())
   B=ScalarParameter() {defaultValue=1000, group=NaniteWPO, parameterName=DistanceDivisor}
  ) {constB=1000}
  n28 = Constant()
  n29 = Multiply(
   A=ScalarParameter() {defaultValue=0.25, group=NaniteWPO, parameterName=Scaling_Far}
   B=MaterialFunctionCall() {materialFunction=&ObjectScale}.Scale X
  )
  n30 = ScalarParameter() {defaultValue=0.5, group=NaniteWPO, parameterName=Scaling_Close}
  n31 = MaterialFunctionCall() {materialFunction=&ObjectScale}
  n32 = Divide(A=Distance(A=CameraPositionWS(), B=NamedRerouteUsage())) {constB=1000}
  n33 = Constant()
  n34 = Multiply(
   A=ScalarParameter() {defaultValue=0.6, group=NaniteWPO, parameterName=Scaling_Radius}
   B=MaterialFunctionCall() {materialFunction=&ObjectScale}.Scale X
  )
  n35 = Sine(None=Multiply(A=@n22.R, B=Add(A=@n22.G, B=@n22.G)))
  n36 = ScalarParameter() {defaultValue=659.55, group=NaniteWPO, parameterName=RandomSeed}
  n37 = Frac(None=Subtract(A=Multiply(A=@n35, B=@n36)) {constB=0.694})
  n38 = ScalarParameter() {defaultValue=0.15, group=NaniteWPO, parameterName=UnhideLeafAmount}
  n39 = LinearInterpolate(
   A=@n29
   B=LinearInterpolate(A=@n29, B=Multiply(A=@n30, B=@n31.Scale X), Alpha=SphereMask(A=@n32, B=@n33, Radius=@n34) {attenuationRadius=0.3, hardnessPercent=0})
   Alpha=OneMinus(None=Step(Y=@n37, X=@n38) {constX=0.1, constY=0.85})
  )
  n40 = SphereMask(A=@n27, B=@n28, Radius=@n39) {attenuationRadius=0.1, hardnessPercent=50}
  n41 = SmoothStep(
   Max=ScalarParameter() {defaultValue=0.5, group=NaniteWPO, parameterName=LeafSizeMin}
   Value=LinearInterpolate(
    A=SphereMask(A=@n27, B=@n28, Radius=@n39) {attenuationRadius=0.1, hardnessPercent=0.8}
    B=Power(Base=@n40) {constExponent=0.5}
    Alpha=NamedRerouteUsage()
   )
  ) {constMax=0.5}
  n42 = Step(Y=@n37) {constX=0, constY=0.85}
  NamedRerouteDeclaration(None=Multiply(A=Multiply(A=@n24, B=@n41), B=OneMinus(None=@n42))) {name=WPO, nodeColor=<rgba 0 1 0.3438 1>}
  n43 = LinearInterpolate(
   A=Multiply(A=NamedRerouteUsage(), B=ScalarParameter() {defaultValue=1, group=SubSurface Settings, parameterName=SSSColor})
   B=Constant()
   Alpha=NamedRerouteUsage()
  )
  n44 = NamedRerouteUsage()
  n45 = LinearInterpolate(A=ScalarParameter() {defaultValue=0.5, group=SubSurface Settings, parameterName=SSSBrightnessBase}, Alpha=@n44)
  NamedRerouteDeclaration(None=StaticSwitchParameter(True=Multiply(A=@n43, B=@n44), False=Multiply(A=@n43, B=@n45)) {parameterName=Is2DMaskMaterial}) {name=SSSColor, nodeColor=<rgba 1 0.0938 0 1>}
  n46 = ScalarParameter() {defaultValue=1, group=SubSurface Settings, parameterName=SSSOpacity}
  n47 = ScalarParameter() {defaultValue=0.5, group=SubSurface Settings, parameterName=SSSOpacity2}
  n48 = VertexColor()
  NamedRerouteDeclaration(None=Saturate(None=LinearInterpolate(A=@n46, B=@n47, Alpha=@n48.B))) {name=SSSOpacity, nodeColor=<rgba 0 0.3672 1 1>}
  NamedRerouteDeclaration(None=VertexColor().B) {name=VerticalGradient, nodeColor=<rgba 1 0 0.0703 1>}
  NamedRerouteDeclaration(None=Frac(None=VertexColor().G)) {name=RandomPerBlade, nodeColor=<rgba 0.7109 0 1 1>}
  n49 = TextureSampleParameter2D() {
   group=Textures
   parameterName=Mask2Dtexture
   samplerType=SAMPLERTYPE_Masks
   texture=&T_WhiteMask
   channelNames {a="", b=RandomPerBlade, g=BakedAO, r=OpacityMask}
  }
  NamedRerouteDeclaration(None=@n49.RandomPerBlade) {name=RandomPerBlade2d, nodeColor=<rgba 0.0312 1 0 1>}
  NamedRerouteDeclaration(None=TextureSample() {samplerType=SAMPLERTYPE_Normal, texture=/Game/Environments/Asteria/Foliage/Grass/Textures/Grass/CylinderNormal_N}.RGB) {name=CylinderNormal, nodeColor=<rgba 1 0.9375 0 1>}
  n50 = TextureCoordinate() {coordinateIndex=1}
  n51 = MaterialFunctionCall(X=ComponentMask(None=@n50) {r=true}, Y=ComponentMask(None=@n50) {g=true}, Z=Constant()) {materialFunction=&MakeFloat3}
  NamedRerouteDeclaration(None=VertexInterpolator(VS=TransformPosition(None=@n51.Result) {transformSourceType=TRANSFORMPOSSOURCE_Instance, transformType=TRANSFORMPOSSOURCE_World}).PS) {name=PivotForRVT, nodeColor=<rgba 1 0.75 0 1>}
  n52 = MaterialFunctionCall!(UseLandscapeCoords) {materialFunction=/Game/Athena/Asteria/Environments/Landscape/Materials/MaterialFunctions/MF_AsteriaTerrainMasks}
  NamedRerouteDeclaration(None=StaticSwitchParameter(
    True=MaterialFunctionCall!(BiomeMask-RGBA=@n52.BiomeMaskRGBA, TextureSamplingPosition, IsGrass) {materialFunction=/Game/Athena/Asteria/Environments/Landscape/Materials/MaterialFunctions/MF_AsteriaGrassColors}.NORVT_GrassColors
    False=VectorParameter() {group=Color, parameterName=GrassColor_FallBack, defaultValue=<rgba 0.0471 0.149 0.0471 1>}.RGB
   ) {defaultValue=true, parameterName="UseGrassColors(BiomeTexture)"}) {name=LandscapeColor_Influence, nodeColor=<rgba 1 0.9375 0 1>}
  NamedRerouteDeclaration(None=@n49.OpacityMask) {name=Opacity Mask, nodeColor=<rgba 0.1016 0 1 1>}
  NamedRerouteDeclaration(None=@n2.Roughness) {name=Roughness Texture, nodeColor=<rgba 0.5938 0 1 1>}
  NamedRerouteDeclaration(None=@n49.BakedAO) {name=Baked AO, nodeColor=<rgba 1 0 0.6797 1>}
  n53 = NamedRerouteUsage()
  n54 = NamedRerouteUsage()
  n55 = StaticSwitchParameter(True=NamedRerouteUsage(), False=VertexColor().G) {parameterName=Is2DMaskMaterial}
  n56 = Multiply(A=NamedRerouteUsage(), B=LinearInterpolate(A=ScalarParameter() {defaultValue=1, group=Color, parameterName=ValueVar}, Alpha=@n55))
  n57 = StaticSwitchParameter(
   True=@n56
   False=MaterialFunctionCall(Hue Shift Percentage=NamedRerouteUsage(), Texture=@n56) {materialFunction=&HueShift}.Result
  ) {parameterName=IsFlowers}
  n58 = NamedRerouteUsage()
  n59 = NamedRerouteDeclaration(None=FeatureLevelSwitch(Default=QualitySwitch(Default=@n57, Low=@n58, Medium=@n58), ES3_1=@n58)) {name=BaseColor, nodeColor=<rgba 0 1 0.5078 1>}
  n60 = ScalarParameter() {group=Color, parameterName=BakedAOintensity}
  n61 = NamedRerouteUsage()
  n62 = Multiply(A=@n59, B=Saturate(None=LinearInterpolate(A=@n60, Alpha=@n61)))
  n63 = FeatureLevelSwitch(
   Default=QualitySwitch(Default=StaticSwitchParameter(True=@n62, False=@n59) {parameterName=Is2DMaskMaterial}, Low=@n59, Medium=@n59)
   ES3_1=@n59
  )
  NamedRerouteDeclaration(None=StaticSwitchParameter(True=LinearInterpolate(A=@n63, B=@n53, Alpha=@n54), False=@n63) {group=FlowerFade, parameterName=UseFlowerFadeToGreen}) {name=Final BaseColor, nodeColor=<rgba 1 0 0.8203 1>}
  n64 = TextureSampleParameter2D() {
   group=Textures
   parameterName=Diffuse
   samplerSource=SSM_Wrap_WorldGroupSettings
   texture=/Game/Characters/Fortress_Character/Textures/T_All_White
  }
  n65 = Multiply(A=NamedRerouteUsage(), B=@n64.RGB)
  n66 = NamedRerouteUsage()
  n67 = Multiply(A=Subtract(A=PerInstanceRandom()) {constB=0.5}, B=ScalarParameter() {defaultValue=-0.024, group=Color, parameterName=HueVarPerCluster})
  n68 = VectorParameter() {group=Color, parameterName=ColorTint, defaultValue=<rgba 1 1 1 1>}
  n69 = StaticSwitchParameter(
   True=MaterialFunctionCall(Hue Shift Percentage=Add(A=@n66, B=@n67), Texture=Multiply(A=@n64.RGB, B=@n68.RGB)) {materialFunction=&HueShift}.Result
   False=@n64.RGB
  ) {parameterName=IsFlowers}
  n70 = NamedRerouteUsage()
  NamedRerouteDeclaration(None=StaticSwitchParameter(True=LinearInterpolate(A=@n65, B=@n69, Alpha=@n70), False=NamedRerouteUsage()) {group=Textures, parameterName=UseTextures}) {name=ColorLowQuality, nodeColor=<rgba 0.1953 0 1 1>}
  n71 = RuntimeVirtualTextureSampleParameter!(UVs, World Position=NamedRerouteUsage(), Mip Level=Constant()) {
   bEnableFeedback=false
   group=RVT
   materialType=BaseColor_Normal_Specular_YCoCg
   mipValueMode=RVTMVM_MipLevel
   parameterName=RVT Sample Target
   virtualTexture=/Game/Athena/Apollo/Maps/Landscape/VT/RVT_Apollo_Landscape
  }
  n72 = NamedRerouteUsage()
  n73 = MaterialFunctionCall(NoRVT=@n72, RVT=StaticSwitchParameter(True=@n71.BaseColor, False=@n72) {defaultValue=true, parameterName=UseRVT}) {materialFunction=&FN_RVT_Switch_Material}
  NamedRerouteDeclaration(None=FeatureLevelSwitch(Default=QualitySwitch(Default=@n73.Result, Low=@n72), ES2=@n72, ES3_1=@n72)) {name=RVTorBiomeColor, nodeColor=<rgba 1 0.5391 0 1>}
  NamedRerouteDeclaration(None=@n2.ColorSystemMask) {name=ColorSytemTexturemask, nodeColor=<rgba 0.7813 0 1 1>}
  NamedRerouteDeclaration(None=Multiply(A=Subtract(A=@n55) {constB=0.5}, B=ScalarParameter() {defaultValue=-0.0832, group=Color, parameterName=HueVarPerBlade})) {name=HueVarPer Blade, nodeColor=<rgba 0 0.8594 1 1>}
  NamedRerouteDeclaration(None=TransformPosition(None=Constant3Vector()) {transformSourceType=TRANSFORMPOSSOURCE_Instance, transformType=TRANSFORMPOSSOURCE_World}) {name=ActorLocation, nodeColor=<rgba 0 0.2734 1 1>}
  n74 = PixelDepth()
  n75 = Add(A=@n74.R, B=ScalarParameter() {defaultValue=-2000, group=Camera Fade Distances, parameterName=NearFadeOffset})
  n76 = ScalarParameter() {defaultValue=5000, group=Camera Fade Distances, parameterName=NearFade}
  NamedRerouteDeclaration(None=Saturate(None=Divide(A=@n75, B=@n76))) {name=NearFadeMask, nodeColor=<rgba 1 0 0.9609 1>}
  n77 = Add(A=@n74.R, B=ScalarParameter() {defaultValue=-7000, group=Camera Fade Distances, parameterName=FarFadeOffset})
  n78 = ScalarParameter() {defaultValue=4000, group=Camera Fade Distances, parameterName=FarFade}
  NamedRerouteDeclaration(None=Saturate(None=Divide(A=@n77, B=@n78))) {name=FarFadeMask, nodeColor=<rgba 1 0 0.8672 1>}
  n79 = VertexColor()
  NamedRerouteDeclaration(None=Abs(None=Sign(None=@n79))) {name=CardMask, nodeColor=<rgba 1 0 0.375 1>}
 ]
}
